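; Test actor that previews Niagara aura effects on nine guitar, singing and drum soldier mannequins, male and female.
; BeginPlay starts four async-load chains, each loading character definitions one after another.
; Each completed load applies its character to a child-actor FortPlayerMannequin and attaches that slot's particle system to the skeletal mesh.
; The custom ResetParticles event sets every slot's particle variable to Default_Particles, without respawning effects.
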(event EventBeginPlay
  (Utilities|FlowControl|Sequence
    (:then_0
      ;[1 1 1] Guitar Male 01
      (bind _object (Utilities|AsyncLoadAsset (Variables|Default|GetGuitar_Male_01))
        (:Completed
          (bind _asathena_character_item_definition (Utilities|Casting|CastToAthenaCharacterItemDefinition _object))
          (bind _asfort_player_mannequin (Utilities|Casting|CastToFortPlayerMannequin (Variables|/Script/Engine.ChildActorComponent|GetChildActor (Variables|Default|GetGuitar_SoldierMannequin_Male_01))))
          (Customization|ResetAthenaCharacterDefandVariants _asfort_player_mannequin _asathena_character_item_definition (Utilities|Array|MakeArray))
          (Niagara|SpawnSystemAttached (Variables|Default|GetGuitar_Male_01_Particles) (Variables|/Script/Engine.SkeletalMeshActor|GetSkeletalMeshComponent _asfort_player_mannequin) "")
          ;[1 1 1] Sing Male 01
          (bind _object_2 (Utilities|AsyncLoadAsset (Variables|Default|GetSing_Male_01))
            (:Completed
              (bind _asathena_character_item_definition_2 (Utilities|Casting|CastToAthenaCharacterItemDefinition _object_2))
              (bind _asfort_player_mannequin_2 (Utilities|Casting|CastToFortPlayerMannequin (Variables|/Script/Engine.ChildActorComponent|GetChildActor (Variables|Default|GetSing_SoldierMannequin_Male_01))))
              (Customization|ResetAthenaCharacterDefandVariants _asfort_player_mannequin_2 _asathena_character_item_definition_2 (Utilities|Array|MakeArray))
              (Niagara|SpawnSystemAttached (Variables|Default|GetSing_Male_01_Particles) (Variables|/Script/Engine.SkeletalMeshActor|GetSkeletalMeshComponent _asfort_player_mannequin_2) "")
              ;[1 1 1] Drum
              (bind _object_3 (Utilities|AsyncLoadAsset (Variables|Default|GetDrum_Male_01))
                (:Completed
                  (bind _asathena_character_item_definition_3 (Utilities|Casting|CastToAthenaCharacterItemDefinition _object_3))
                  (bind _asfort_player_mannequin_3 (Utilities|Casting|CastToFortPlayerMannequin (Variables|/Script/Engine.ChildActorComponent|GetChildActor (Variables|Default|GetDrum_SoldierMannequin_Male_01))))
                  (Customization|ResetAthenaCharacterDefandVariants _asfort_player_mannequin_3 _asathena_character_item_definition_3 (Utilities|Array|MakeArray))
                  (Niagara|SpawnSystemAttached (Variables|Default|GetDrum_Male_01_Particles) (Variables|/Script/Engine.SkeletalMeshActor|GetSkeletalMeshComponent _asfort_player_mannequin_3) ""))))))))
    (:then_1
      ;[1 1 1] Guitar Female 01
      (bind _object_4 (Utilities|AsyncLoadAsset (Variables|Default|GetGuitar_Female_01))
        (:Completed
          (bind _asathena_character_item_definition_4 (Utilities|Casting|CastToAthenaCharacterItemDefinition _object_4))
          (bind _asfort_player_mannequin_4 (Utilities|Casting|CastToFortPlayerMannequin (Variables|/Script/Engine.ChildActorComponent|GetChildActor (Variables|Default|GetGuitar_SoldierMannequin_Female_01))))
          (Customization|ResetAthenaCharacterDefandVariants _asfort_player_mannequin_4 _asathena_character_item_definition_4 (Utilities|Array|MakeArray))
          (Niagara|SpawnSystemAttached (Variables|Default|GetGuitar_Female_01_Particles) (Variables|/Script/Engine.SkeletalMeshActor|GetSkeletalMeshComponent _asfort_player_mannequin_4) "")
          ;[1 1 1] Sing Female 01
          (bind _object_5 (Utilities|AsyncLoadAsset (Variables|Default|GetSing_Female_01))
            (:Completed
              (bind _asathena_character_item_definition_5 (Utilities|Casting|CastToAthenaCharacterItemDefinition _object_5))
              (bind _asfort_player_mannequin_5 (Utilities|Casting|CastToFortPlayerMannequin (Variables|/Script/Engine.ChildActorComponent|GetChildActor (Variables|Default|GetSing_SoldierMannequin_Female_01))))
              (Customization|ResetAthenaCharacterDefandVariants _asfort_player_mannequin_5 _asathena_character_item_definition_5 (Utilities|Array|MakeArray))
              (Niagara|SpawnSystemAttached (Variables|Default|GetSing_Female_01_Particles) (Variables|/Script/Engine.SkeletalMeshActor|GetSkeletalMeshComponent _asfort_player_mannequin_5) "")
              ;[1 1 1] Drum
              (bind _object_6 (Utilities|AsyncLoadAsset (Variables|Default|GetDrum_Female_01))
                (:Completed
                  (bind _asathena_character_item_definition_6 (Utilities|Casting|CastToAthenaCharacterItemDefinition _object_6))
                  (bind _asfort_player_mannequin_6 (Utilities|Casting|CastToFortPlayerMannequin (Variables|/Script/Engine.ChildActorComponent|GetChildActor (Variables|Default|GetDrum_SoldierMannequin_Female_01))))
                  (Customization|ResetAthenaCharacterDefandVariants _asfort_player_mannequin_6 _asathena_character_item_definition_6 (Utilities|Array|MakeArray))
                  (Niagara|SpawnSystemAttached (Variables|Default|GetDrum_Female_01_Particles) (Variables|/Script/Engine.SkeletalMeshActor|GetSkeletalMeshComponent _asfort_player_mannequin_6) ""))))))))
    (:then_2
      ;[1 1 1] Guitar Male 02
      (bind _object_7 (Utilities|AsyncLoadAsset (Variables|Default|GetGuitar_Male_02))
        (:Completed
          (bind _asathena_character_item_definition_7 (Utilities|Casting|CastToAthenaCharacterItemDefinition _object_7))
          (bind _asfort_player_mannequin_7 (Utilities|Casting|CastToFortPlayerMannequin (Variables|/Script/Engine.ChildActorComponent|GetChildActor (Variables|Default|GetGuitar_SoldierMannequin_Male_02))))
          (Customization|ResetAthenaCharacterDefandVariants _asfort_player_mannequin_7 _asathena_character_item_definition_7 (Utilities|Array|MakeArray))
          (Niagara|SpawnSystemAttached (Variables|Default|GetGuitar_Male_02_Particles) (Variables|/Script/Engine.SkeletalMeshActor|GetSkeletalMeshComponent _asfort_player_mannequin_7) "")
          ;[1 1 1] Sing Male 02
          (bind _object_8 (Utilities|AsyncLoadAsset (Variables|Default|GetSing_Male_02))
            (:Completed
              (bind _asathena_character_item_definition_8 (Utilities|Casting|CastToAthenaCharacterItemDefinition _object_8))
              (bind _asfort_player_mannequin_8 (Utilities|Casting|CastToFortPlayerMannequin (Variables|/Script/Engine.ChildActorComponent|GetChildActor (Variables|Default|GetSing_SoldierMannequin_Male_02))))
              (Customization|ResetAthenaCharacterDefandVariants _asfort_player_mannequin_8 _asathena_character_item_definition_8 (Utilities|Array|MakeArray))
              (Niagara|SpawnSystemAttached (Variables|Default|GetSing_Male_02_Particles) (Variables|/Script/Engine.SkeletalMeshActor|GetSkeletalMeshComponent _asfort_player_mannequin_8) ""))))))
    (:then_3
      ;[1 1 1] Guitar Female 02
      (bind _object_9 (Utilities|AsyncLoadAsset (Variables|Default|GetGuitar_Female_02))
        (:Completed
          (bind _asathena_character_item_definition_9 (Utilities|Casting|CastToAthenaCharacterItemDefinition _object_9))
          (bind _asfort_player_mannequin_9 (Utilities|Casting|CastToFortPlayerMannequin (Variables|/Script/Engine.ChildActorComponent|GetChildActor (Variables|Default|GetGuitar_SoldierMannequin_Female_02))))
          (Customization|ResetAthenaCharacterDefandVariants _asfort_player_mannequin_9 _asathena_character_item_definition_9 (Utilities|Array|MakeArray))
          (Niagara|SpawnSystemAttached (Variables|Default|GetGuitar_Female_02_Particles) (Variables|/Script/Engine.SkeletalMeshActor|GetSkeletalMeshComponent _asfort_player_mannequin_9) ""))))))

(event Custom|ResetParticles
  (bind _default_particles (Variables|Default|GetDefault_Particles))
  (Utilities|FlowControl|Sequence
    (:then_0
      (Variables|Default|SetGuitar_Female_01_Particles _default_particles)
      (Variables|Default|SetGuitar_Male_01_Particles _default_particles)
      (Variables|Default|SetGuitar_Female_02_Particles _default_particles)
      (Variables|Default|SetGuitar_Male_02_Particles _default_particles))
    (:then_1
      (Variables|Default|SetSing_Male_02_Particles _default_particles)
      (Variables|Default|SetSing_Female_01_Particles _default_particles)
      (Variables|Default|SetSing_Male_01_Particles _default_particles))
    (:then_2
      (Variables|Default|SetDrum_Male_01_Particles _default_particles)
      (Variables|Default|SetDrum_Female_01_Particles _default_particles))))

(fn ConstructionScript)
