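; Test actor that previews aura Niagara effects on a single small-body Fortnite player mannequin held in a child actor.
; BeginPlay async-loads the Character item definition, resets the mannequin to it without variants, and attaches Aura VFX to its skeletal mesh.
; The Trigger_Start/Stop_Particles event repeats that flow with Aura Start VFX; nothing here deactivates earlier systems.
; Both paths set IsLargeBody false and pass the IsFrontEnd flag to Niagara.
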(event Custom|Trigger_Start/Stop_Particles
  ;[1 1 1] Start/Stop Trigger FX
  (bind _object (Utilities|AsyncLoadAsset (Variables|Default|GetCharacter))
    (:Completed
      (bind _asathena_character_item_definition (Utilities|Casting|CastToAthenaCharacterItemDefinition _object))
      (bind _asfort_player_mannequin (Utilities|Casting|CastToFortPlayerMannequin (Variables|/Script/Engine.ChildActorComponent|GetChildActor (Variables|Default|GetGuitar_SoldierMannequin_Female_01))))
      (Customization|ResetAthenaCharacterDefandVariants _asfort_player_mannequin _asathena_character_item_definition (Utilities|Array|MakeArray))
      (bind _returnvalue (Niagara|SpawnSystemAttached (Variables|Default|GetAura%20Start%20VFX) (Variables|/Script/Engine.SkeletalMeshActor|GetSkeletalMeshComponent _asfort_player_mannequin) ""))
      (Niagara|SetNiagaraVariableByString(Bool) _returnvalue "IsLargeBody" false)
      (Niagara|SetNiagaraVariableByString(Bool) _returnvalue "IsFrontEnd" (Variables|Default|GetIsFrontEnd?)))))

(event EventBeginPlay
  ;[1 1 1] Guitar Female 01
  (bind _object (Utilities|AsyncLoadAsset (Variables|Default|GetCharacter))
    (:Completed
      (bind _asathena_character_item_definition (Utilities|Casting|CastToAthenaCharacterItemDefinition _object))
      (bind _asfort_player_mannequin (Utilities|Casting|CastToFortPlayerMannequin (Variables|/Script/Engine.ChildActorComponent|GetChildActor (Variables|Default|GetGuitar_SoldierMannequin_Female_01))))
      (Customization|ResetAthenaCharacterDefandVariants _asfort_player_mannequin _asathena_character_item_definition (Utilities|Array|MakeArray))
      (bind _returnvalue (Niagara|SpawnSystemAttached (Variables|Default|GetAura%20VFX) (Variables|/Script/Engine.SkeletalMeshActor|GetSkeletalMeshComponent _asfort_player_mannequin) ""))
      (Niagara|SetNiagaraVariableByString(Bool) _returnvalue "IsLargeBody" false)
      (Niagara|SetNiagaraVariableByString(Bool) _returnvalue "IsFrontEnd" (Variables|Default|GetIsFrontEnd?)))))

(fn ConstructionScript)
